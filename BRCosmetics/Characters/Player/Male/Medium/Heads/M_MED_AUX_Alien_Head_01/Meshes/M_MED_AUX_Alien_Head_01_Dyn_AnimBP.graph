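# Animation Blueprint for the alien fish head cosmetic, driving face and secondary motion on the player.
# It copies the owner mesh's pose and curves, layers alien_idle from the head bone, applies a facial PoseAsset, then adds soft head spring dynamics.
# When the playlist uses custom character parts, checked at initialization, it shrinks neck_01 to near zero.
# Tail and fin dynamics and an idle test player sit unconnected to the output.
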AnimBlueprint /BRCosmetics/Characters/Player/Male/Medium/Heads/M_MED_AUX_Alien_Head_01/Meshes/M_MED_AUX_Alien_Head_01_Dyn_AnimBP
target_skeleton=/Game/Characters/Player/Common/Fortnite_Base_Head/Fortnite_Base_Head_Export_Skeleton
graphs [
 AnimGraph {nodes [
   n1 = Layered blend per bone!(
    BasePose=Use cached pose 'Copy Pose'()
    BlendPoses_0
    BlendPoses_1:Pose Link Structure=Use cached pose 'Copy Pose'()
    BlendWeights_0
    BlendWeights_1:float=1
   ) -> (Pose) {
    blendPoses [{linkId=-1, sourceLinkId=-1}, {linkId=-1, sourceLinkId=-1}]
    blendWeights [1,1]
    layerSetup [{branchFilters [{boneName=head}]}, {branchFilters [{blendDepth=1, boneName=faceAttach}]}]
   }
   n2 = Anim Dynamics - Bone: dyn_tail_01(ComponentPose=Local To Component(LocalPose=@n1), Alpha=1, ExternalForce="0,0,0") {
    bGravityOverrideInSimSpace=true
    alphaBoolBlend {blendOption=Linear}
    boundBone {boneName=dyn_tail_01}
    physicsBodyDefinitions [{
      sphereCollisionRadius=0
      boundBone {boneName=dyn_tail_01}
      boxExtents=<1 1 1>
      constraintSetup {angularLimitsMax=<0 0 30>, angularLimitsMin=<0 0 -30>, angularTarget=<0 0 0>}
      localJointOffset=<0 0 0>
     }]
   }
   n3 = Anim Dynamics - Bone: dyn_fin_l(ComponentPose=@n2, Alpha=1, ExternalForce="0,0,0") {
    bGravityOverrideInSimSpace=true
    alphaBoolBlend {blendOption=Linear}
    boundBone {boneName=dyn_fin_l}
    physicsBodyDefinitions [{
      sphereCollisionRadius=0
      boundBone {boneName=dyn_fin_l}
      boxExtents=<1 1 1>
      constraintSetup {angularLimitsMax=<0 0 15>, angularLimitsMin=<0 0 -15>, angularTarget=<0 0 0>}
      localJointOffset=<0 0 0>
     }]
   }
   n4 = Anim Dynamics - Bone: dyn_fin_r(ComponentPose=@n3, Alpha=1, ExternalForce="0,0,0") {
    bGravityOverrideInSimSpace=true
    alphaBoolBlend {blendOption=Linear}
    boundBone {boneName=dyn_fin_r}
    physicsBodyDefinitions [{
      sphereCollisionRadius=0
      boundBone {boneName=dyn_fin_r}
      boxExtents=<1 1 1>
      constraintSetup {angularLimitsMax=<0 0 15>, angularLimitsMin=<0 0 -15>, angularTarget=<0 0 0>}
      localJointOffset=<0 0 0>
     }]
   }
   Component To Local(ComponentPose=@n4)
   n5 = "Transform (Modify) Bone - Bone: neck_01"!(ComponentPose=Local To Component(LocalPose=Use cached pose 'PreScaleCache'()), Alpha=1) -> (Pose) {scaleMode=BMM_Replace, alphaBoolBlend {blendOption=Linear}, boneToModify {boneName=neck_01}, scale=<0.01 0.01 0.01>}
   Output Pose(Result=Blend Poses by bool(
     BlendPose_0=Component To Local(ComponentPose=@n5)
     BlendPose_1=Use cached pose 'PreScaleCache'()
     BlendTime_0=0
     BlendTime_1=0
     bActiveValue=Get GetPlayerListUsesCustomCharacterParts() -> (GetPlayerListUsesCustomCharacterParts:bool)
    ) {blendType=Linear}) {layerGroup=None}
   n6 = Get MeshToCopy() -> (MeshToCopy:Skeletal Mesh Component Object Reference)
   n7 = M_MED_AUX_Alien_Head_01_Facial_Pose_Export_PoseAsset!(SourcePose:Pose Link Structure=Layered blend per bone!(
     BasePose=Copy Pose From Mesh(SourceMeshComponent=@n6) {bCopyCurves=true}
     BlendPoses_0=Sequence Player 'alien_idle'() {sequence=/BRCosmetics/Animation/Game/MainPlayer/Facial/Alien_Fish_Head/alien_idle}
     BlendWeights_0
    ) -> (Pose) {layerSetup [{branchFilters [{blendDepth=1, boneName=head}]}]}) -> (Pose:Pose Link Structure) {
    is=title AnimGraphNode_PoseBlendNode
    blendOption=Linear
    poseAsset=&/M_MED_AUX_Alien_Head_01_Facial_Pose_Export_PoseAsset
   }
   n8 = Anim Dynamics - Bone: head(ComponentPose=Local To Component(LocalPose=@n7), Alpha=0.125, ExternalForce="0,0,0") {
    bAngularSpring=true
    bGravityOverrideInSimSpace=true
    bLinearSpring=true
    gravityScale=0.25
    linearDampingOverride=10
    linearSpringConstant=15
    alphaBoolBlend {blendOption=Linear}
    boundBone {boneName=head}
    physicsBodyDefinitions [{
      sphereCollisionRadius=0
      boundBone {boneName=head}
      boxExtents=<120 60 60>
      constraintSetup {angularLimitsMax=<0 0 0>, angularLimitsMin=<0 0 0>, angularTarget=<0 0 0>, linearAxesMax=<0 0 9>, linearAxesMin=<0 0 -9>}
      localJointOffset=<0 0 0>
     }]
   }
   Save cached pose 'PreScaleCache'(Pose=Component To Local(ComponentPose=@n8)) {cacheName=PreScaleCache}
   Sequence Player 'M_MED_AUX_Alien_Head_01_idleTest'() {sequence=&/M_MED_AUX_Alien_Head_01_idleTest}
  ]}
 EventGraph {nodes [
   Event BlueprintInitializeAnimation()
   > Set MeshToCopy(MeshToCopy:Skeletal Mesh Component Object Reference=GetOwnerSkeletalMeshComponent()) -> (Output_Get:Skeletal Mesh Component Object Reference)
   > Set GetPlayerListUsesCustomCharacterParts(GetPlayerListUsesCustomCharacterParts:bool=GetPlaylistUsesCustomCharacterParts!() -> (ReturnValue:bool)) -> (Output_Get:bool)
  ]}
]
variables [
 GetPlayerListUsesCustomCharacterParts {type=boolean}
 MeshToCopy {type=SkeletalMeshComponent}
]
